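# Sound cue that plays the GrillMoon pre-fight blast sequence audio, apparently for an event cinematic moment.
# A single wave player feeds the root directly, slightly reduced in volume and routed to the Fort_Events sound class.
# Attenuation is overridden to disable distance falloff, spatialization and reverb send, so it plays as non-positional 2D audio.
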SoundCue /GrillMoon/Assets/Audio/swb/SC_GrillMoon_swb_PreFight_Blast
root=WavePlayer_0
nodes [{id=WavePlayer_0, properties {soundWaveAssetPtr=/GrillMoon/Assets/Audio/wav/GrillMoon_Seq_PreFight_Blast}}]
properties {
 bOverrideAttenuation=true
 maxDistance=2097152
 soundClassObject=&Fort_Events
 volumeMultiplier=0.8
 attenuationOverrides {bAttenuate=false, bEnableReverbSend=false, bSpatialize=false}
}
