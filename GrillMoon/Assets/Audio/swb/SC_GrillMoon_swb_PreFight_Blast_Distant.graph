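# Sound cue for the distant layer of the GrillMoon pre-fight blast, playing the distant portal buildup wave in the events sound class.
# A distance crossfade leaves the near input empty and fades the wave in at long range, so it is silent close to the source.
# An attenuation node sets low-pass and occlusion parameters.
# Volume attenuation and reverb send are off at node and cue level, and the cue also disables spatialization.
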SoundCue /GrillMoon/Assets/Audio/swb/SC_GrillMoon_swb_PreFight_Blast_Distant
root=DistanceCrossFade_1
nodes [
 {
  id=Attenuation_0
  children [WavePlayer_2]
  properties {
   bOverrideAttenuation=true
   attenuationOverrides {
    bAttenuate=false
    bEnableReverbSend=false
    lPFFrequencyAtMax=8000
    lPFFrequencyAtMin=4000
    lPFRadiusMax=10000000
    lPFRadiusMin=16000
    occlusionLowPassFilterFrequency=1200
    occlusionTraceChannel=ECC_GameTraceChannel6
   }
  }
 }
 {
  id=DistanceCrossFade_1
  children [null, Attenuation_0]
  properties {crossFadeInput [
    {fadeInDistanceEnd=0, fadeInDistanceStart=0, fadeOutDistanceEnd=24000, fadeOutDistanceStart=16000, volume=1}
    {fadeInDistanceEnd=24000, fadeInDistanceStart=16000, fadeOutDistanceEnd=10000000, fadeOutDistanceStart=10000000, volume=1}
   ]}
 }
 {id=WavePlayer_2, properties {soundWaveAssetPtr=/GrillMoon/Assets/Audio/wav/GrillMoon_Seq_PortalBuildup_Distant}}
]
properties {
 bOverrideAttenuation=true
 maxDistance=2097152
 soundClassObject=&Fort_Events
 volumeMultiplier=0.8
 attenuationOverrides {bAttenuate=false, bEnableReverbSend=false, bSpatialize=false}
}
